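# Physics asset for the F_MED_Division backpack cosmetic mesh, simulating its dangling zippers and knives.
# Small capsule bodies sit on the root bone and on left and right dyn_knife and dyn_zipper bones.
# Each dyn bone is constrained directly to root with swing and twist limits, looser for the zippers and tighter for the knives, so the zippers swing more freely.
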PhysicsAsset /BRCosmetics/Accessories/FORT_Backpacks/Backpack_F_MED_Division/Meshes/F_MED_Division_Pack_Physics
bodies {
 dyn_knifeLf [{shapeType=Capsule, length=2, radius=2, shapeName=dyn_knifeLf_capsule, center=<1.1816 0.0002 0>, rotation=<rot 90 0 0>}]
 dyn_knifeRt [{shapeType=Capsule, length=2, radius=2, shapeName=dyn_knifeRt_capsule, center=<-1.3939 -0.0001 0>, rotation=<rot 90 0 0>}]
 dyn_zipperLf [{shapeType=Capsule, length=2, radius=2, shapeName=dyn_zipperLf_capsule, center=<6.4036 0.0001 0.0002>, rotation=<rot 90 0 0>}]
 dyn_zipperRt [{shapeType=Capsule, length=2, radius=2, shapeName=dyn_zipperRt_capsule, center=<-7.8743 0 0.0002>, rotation=<rot 90 0 0>}]
 root [{shapeType=Capsule, length=0.505, radius=0.505, shapeName=root_capsule, rotation=<rot 0 -90 0>}]
}
constraints [
 {bone1Name=dyn_zipperLf, bone2Name=root, swing1LimitDegrees=40, swing2LimitDegrees=40, twistLimitDegrees=10}
 {bone1Name=dyn_zipperRt, bone2Name=root, swing1LimitDegrees=40, swing2LimitDegrees=40, twistLimitDegrees=10}
 {bone1Name=dyn_knifeLf, bone2Name=root, swing1LimitDegrees=10, swing2LimitDegrees=10, twistLimitDegrees=5}
 {bone1Name=dyn_knifeRt, bone2Name=root, swing1LimitDegrees=10, swing2LimitDegrees=10, twistLimitDegrees=5}
]
